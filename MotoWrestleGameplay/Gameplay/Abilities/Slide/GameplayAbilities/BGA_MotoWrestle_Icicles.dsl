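; Icicles gameplay actor for the MotoWrestle slide ability, showing spikes that grow from a player's right foot.
; On BeginPlay it casts its instigator to a Fort player pawn, stores it as OwningPlayer and snaps itself to that pawn's Foot_R socket.
; The GrowSpikes timeline then lerps the actor's relative scale from small to full size.
; The overlap, tick and construction script hooks are empty.
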(event EventBeginPlay
  (bind _asfort_player_pawn (Utilities|Casting|CastToFortPlayerPawn (Variables|Actor|GetInstigator)))
  (bind _output_get (Variables|Default|SetOwningPlayer _asfort_player_pawn))
  (Transformation|AttachActorToActor _output_get "Foot_R" "SnapToTarget" "SnapToTarget" "SnapToTarget" false)
  (bind (_direction _alpha_curve) (Timeline|GrowSpikes Play)
    (:Update
      (Transformation|SetActorRelativeScale3D (Math|Vector|Lerp(Vector) "0.100000,0.100000,0.100000" "1.000000,1.000000,1.000000" _alpha_curve)))))

(event Collision|EventActorBeginOverlap
  (params (OtherActor :type Actor%20Object%20Reference)))

(event EventTick
  (params (DeltaSeconds :type Float%20%28single-precision%29)))

(fn ConstructionScript)
