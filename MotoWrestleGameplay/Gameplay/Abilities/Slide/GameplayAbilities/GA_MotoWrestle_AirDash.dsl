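; MotoWrestle air dash ability that dives the player to the ground, then slides them into a dash attack.
; Guided-motion MMEs drive the dive and the slide (toward a found target or straight), with next-tick continuous targeting traces cutting each short.
; Attack input or jumping exits the slide early, and the impact montage notify applies deduplicated damage.
; Gameplay events mark MME phases and are replayed in order to catch the server up.
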(event Ability|EventActivateAbility
  ;[0.03692 0.15 0.02006 0.5] Activate Ability - Initialize hotfixable failsafe delay
  (Variables|PowerFailsafe|SetFailsafeDelay (ScalableFloat|GetValueatLevel (Variables|Hotfixes|GetHotfix_FailsafeDelay)))
  (|CallParentFunction))

(event EventPreLaunchSetup
  ;[0.1571 0.3406 0.129 0.5] Launch Setup
  (Fortnite|FortShippingLog self "LaunchPlayer: Invoked.")
  (Variables|Transient|SetbLaunchPlayerInvoked true)
  (|CallParentFunction)
  ;;[0.02407 0.124 0.15 0.5] Play the montage that makes sense for diving towards the ground.
  (bind _asynctaskproxy (Ability|Tasks|PlayMontageAndWaitNotify "None" "/MotoWrestleGameplay/Assets/Animations/Slide/MotoWrestle_SlideDash_Air_Montage.MotoWrestle_SlideDash_Air_Montage"))
  (Variables|Transient|Async|SetAsyncTask_PlayMontageSlideDashAir _asynctaskproxy))

(event EventSetupMMEBinds
  ;[0.1571 0.3406 0.129 0.5] Launch Setup
  (|CallParentFunction)
  ;; Set up wait gameplay events to figure out when the AirToGround MME starts and ends.
  (Utilities|FlowControl|Sequence
    (:then_0
      (bind _asynctaskproxy (Ability|Tasks|WaitGameplayEvent "(TagName=\"MotoWrestle.Slide.AirToGround.MME.Start\")")
        (:then
          (Variables|Transient|Async|SetAsyncTask_WaitAirToGroundMMEEventStart _asynctaskproxy))
        (:EventReceived
          (if (not (Variables|Transient|GetAirToGroundMMEStartTriggered))
            (CallFunction|OnGroundtoAirMMEStart)))))
    (:then_1
      (bind _asynctaskproxy_2 (Ability|Tasks|WaitGameplayEvent "(TagName=\"MotoWrestle.Slide.AirToGround.MME.End\")")
        (:then
          (Variables|Transient|Async|SetAsyncTask_WaitAirToGroundMMEEventEnd _asynctaskproxy_2))
        (:EventReceived
          (if (not (Variables|Transient|GetAirToGroundMMEEndTriggered))
            (CallFunction|OnGroundtoAidMMEEnd)))))))

(event EventLaunchPlayer
  ;[0.01438 0.15 0.03199 0.5] LaunchPlayer
  (CallFunction|PreLaunchSetup)
  (Utilities|FlowControl|Sequence
    (:then_0
      ;;[0.15 0.05575 0.1382 0.5] Begin the air to ground constant force MME.
      ;; This is expected to take the player to the ground.
      ;; A hard max duration value is specified, if the player never finds the ground before this duration, the MME will end.
      (Utilities|IsServer
        (:true
          (if (CallFunction|IsAI)
            _))
        (:false
          _))
      (bind _returnvalue (Game|ApplyGuidedMotionMME (CallFunction|GetAvatarasFortPlayerPawn) "/MotoWrestleGameplay/Gameplay/Abilities/Slide/MME/FMM_GuidedMotion_MotoWrestle_AirToGroundDash.FMM_GuidedMotion_MotoWrestle_AirToGroundDash_C"))
      (Variables|Transient|SetAirToGroundMovementModeExtRuntimeData _returnvalue)
      (Fortnite|FortShippingLog self "LaunchPlayer: AirToGround MME invoked."))
    (:then_1
      ;;[0.0259 0.1342 0.15 0.5] On the server, at this point, if we have already received the client's ground dash complete event, then just run through all the MME events in order.
      (Ability|HasAuthority
        (:True
          (if (Variables|Transient|GetbClientGroundDashComplete)
            (CallFunction|ServerNotifyDashMMEComplete)))))
    (:then_2
      (bind _outfortplayerpawn (CallFunction|GetAvatarasValidFortPlayerPawn)
        (:ValidFortPlayerPawn
          (Utilities|IsValid (Fortnite|GetFortPlayerControllerfromActor _outfortplayerpawn)
            (:Is%20Valid
              ;; Camera shake at the start of the slide
              (Game|Feedback|ClientStartCameraShake :self (Fortnite|GetFortPlayerControllerfromActor _outfortplayerpawn) :Shake "/MotoWrestleGameplay/Gameplay/Abilities/Slide/Camera/CameraShake_MotoWrestle_SlideStart.CameraShake_MotoWrestle_SlideStart_C" :Scale 1.0 :PlaySpace "CameraLocal"))))))
    (:then_3
      ;;[0.02138 0.15 0.1496 0.5] GCNL while sliding
      (Ability|AddGameplayCueToOwner (Variables|GameplayCue|GetGC_SlideAudio)))))

(event Custom|On%20Ground%20To%20Air%20MME%20Start
  ;[0.05118 0.15 0.117 0.5] MME Events
  (Variables|Transient|SetAirToGroundMMEStartTriggered true)
  (Variables|Transient|SetbAirToGroundMMERunning true)
  (Fortnite|FortShippingLog self "LaunchPlayer: AirToGround MME started.")
  (bind _returnvalue (Utilities|Time|SetTimerforNextTickbyEvent (EventDispatchers|CreateEvent :!event Custom|BeginAirToGroundTargeting))) ; Optimization to prevent targeting traces from running on the same frame that the MME begins.
  (Variables|Transient|SetBeginAirToGroundFrameDelayTimerHandle _returnvalue)
  ;; Track the frame at which the AirToGroundMME has started.
  (Variables|Transient|SetAirToGroundDurationFrameCount (Utilities|GetFrameCount)))

(event Custom|On%20Ground%20To%20Aid%20MME%20End
  ;[0.05118 0.15 0.117 0.5] MME Events
  (Variables|Transient|SetAirToGroundMMEEndTriggered true)
  (Variables|Transient|SetbAirToGroundMMERunning false)
  (Fortnite|FortShippingLog self "LaunchPlayer: AirToGround MME ended.")
  (Helper|CatchServerUponEventsinOrder (Utilities|Array|MakeArray "(TagName=\"MotoWrestle.Slide.AirToGround.MME.Start\")"))
  (CallFunction|EndAirtoGroundTargeting)
  ;; Ground dash has started from the ground itself it has been just a single frame since the AirToGroundMME started. This is used to decide the starting Montage section to use for the ground dash.
  (CallFunction|BeginGroundDash (<= (- (Utilities|GetFrameCount) (Variables|Transient|GetAirToGroundDurationFrameCount)) 3))
  (bind _airtogroundcameramodifierhandle (Variables|Transient|GetAirToGroundCameraModifierHandle)
    (:then
      (Class|FortCameraModifierTargetRotationFrustum|RemoveSelffromCameraManager _airtogroundcameramodifierhandle))
    (:else)))

(event EventOnMMEStart
  ;[0.05118 0.15 0.117 0.5] MME Events
  (|CallParentFunction)
  (Helper|CatchServerUponEventsinOrder (Utilities|Array|MakeArray "(TagName=\"MotoWrestle.Slide.AirToGround.MME.Start\")" "(TagName=\"MotoWrestle.Slide.AirToGround.MME.End\")"))
  (Camera|ApplyGroundDashCameraModifier)
  (bind _returnvalue (Utilities|Time|SetTimerforNextTickbyEvent (EventDispatchers|CreateEvent :!event Custom|BeginGroundDashTargeting))) ; Optimization to prevent targeting traces from running on the same frame that the MME begins.
  (Variables|Transient|SetBeginTargetingFrameDelayTimerHandle _returnvalue)
  (CallFunction|WaitforAttackInput))

(event EventOnMMEEnd
  ;[0.05118 0.15 0.117 0.5] MME Events
  (|CallParentFunction)
  (Helper|CatchServerUponEventsinOrder (Utilities|Array|MakeArray "(TagName=\"MotoWrestle.Slide.AirToGround.MME.Start\")" "(TagName=\"MotoWrestle.Slide.AirToGround.MME.End\")" "(TagName=\"MotoWrestle.Slide.Ground.MME.Start\")"))
  (CallFunction|EndGroundDashTargeting)
  (Ability|HasAuthority
    (:True
      (if (CallFunction|IsAI)
        _))
    (:False
      _))
  (CallFunction|ServerNotifyDashMMEComplete))

(event Custom|Wait%20For%20Attack%20Input
  (bind _hotfix_inputpressedexecutegrounddashattackmindelay (Variables|Hotfixes|GetHotfix_InputPressedExecuteGroundDashAttackMinDelay))
  (bind _returnvalue (ScalableFloat|GetValueatLevel _hotfix_inputpressedexecutegrounddashattackmindelay))
  ;[0.15 0.1003 0.04318 0.5] Allow players to exit out of the dash into the attack based on their input.
  (if (CallFunction|IsAI)
    (Utilities|FlowControl|Delay (ScalableFloat|GetValueatLevel (Variables|Default|GetHotfix_AIFallBackDuration)))
    (elif (> _returnvalue 0)
      (bind _asynctaskproxy (Ability|Tasks|WaitGameplayEvent "(TagName=\"Event.Primary.Pressed\")")
        (:then
          (Variables|Transient|Async|SetAsyncTask_GroundDashWaitPrimaryPressed _asynctaskproxy)
          (bind _returnvalue_2 (Utilities|Time|SetTimerbyEvent (delegate Custom|InputPressedExecuteGroundDashAttackMinDelayComplete) _returnvalue))
          (Variables|Transient|SetInputPressedExecuteGroundDashAttackMinDelayTimerHandle _returnvalue_2))
        (:EventReceived
          _))))
  (Variables|Transient|SetbInputPressedToExecuteGroundDashAttack true)
  (if (not (Utilities|Time|IsTimerActivebyHandle (Variables|Transient|GetInputPressedExecuteGroundDashAttackMinDelayTimerHandle)))
    (if (and (Variables|Transient|GetbDashMMEStartTriggered) (not (Variables|Transient|GetbDashMMEEndTriggered)))
      (Fortnite|FortShippingLog self "BeginGroundDash: Ending dash cause of attack input.")
      (bind _outfortplayerpawn (CallFunction|GetAvatarasValidFortPlayerPawn)
        (:ValidFortPlayerPawn
          (bind _asfort_movement_comp_character (Utilities|Casting|CastToFortMovementComp_Character (Variables|/Script/Engine.Character|GetCharacterMovement _outfortplayerpawn)))
          (Class|FortMovementCompCharacter|CancelActiveExtensionwithTag _asfort_movement_comp_character (GameplayTags|MakeGameplayTagContainerfromTag "(TagName=\"MovementMode.Custom.RootMotion.MotoWrestle.Slide.Ground\")")))))))

(event Custom|InputPressedExecuteGroundDashAttackMinDelayComplete
  ;[0.15 0.1003 0.04318 0.5] Allow players to exit out of the dash into the attack based on their input.
  (if (Variables|Transient|GetbInputPressedToExecuteGroundDashAttack)
    (if (and (Variables|Transient|GetbDashMMEStartTriggered) (not (Variables|Transient|GetbDashMMEEndTriggered)))
      (Fortnite|FortShippingLog self "BeginGroundDash: Ending dash cause of attack input.")
      (bind _outfortplayerpawn (CallFunction|GetAvatarasValidFortPlayerPawn)
        (:ValidFortPlayerPawn
          (bind _asfort_movement_comp_character (Utilities|Casting|CastToFortMovementComp_Character (Variables|/Script/Engine.Character|GetCharacterMovement _outfortplayerpawn)))
          (Class|FortMovementCompCharacter|CancelActiveExtensionwithTag _asfort_movement_comp_character (GameplayTags|MakeGameplayTagContainerfromTag "(TagName=\"MovementMode.Custom.RootMotion.MotoWrestle.Slide.Ground\")")))))))

(event Custom|BeginAirToGroundTargeting
  ;[0.15 0 0.1498 0.5] Start running air to ground targeting trace.
  ; Used to figure out if the air to ground MME should stop and the ground MME should start.
  ;; BeginAirToGroundTargeting: Start a continuous targeting trace to figure out what is ahead of the player and if the player should crash with it while performing the air to ground dash.
  (bind (_asynctaskproxy _targetdata) (Ability|Tasks|DoTargetSelectionContinuous "AirToGroundContinuousTargeting" "(TagName=\"MotoWrestle.Slide.EffectContainer.Application.AirToGround\")")
    (:then
      (Variables|Transient|Async|SetAsyncTask_ContinuousTargetingAirToGround _asynctaskproxy))
    (:Targeted
      (if (Ability|TargetData|TargetDataHasHitResult _targetdata)
        (Fortnite|FortShippingLog self (Utilities|String|Append "AirToGroundTargeting: Successfully hit actor: " (Utilities|GetDisplayName (.HitActor (Collision|BreakHitResult (Ability|TargetData|GetHitResultfromTargetData _targetdata))))))
        (Ability|EndAirtoGroundMME)))))

(event Custom|EndAirToGroundTargeting
  ;[0.15 0 0.1498 0.5] Start running air to ground targeting trace.
  ; Used to figure out if the air to ground MME should stop and the ground MME should start.
  ;; EndAirToGroundTargeting: End the air to ground continuous targeting.
  (Utilities|Time|ClearandInvalidateTimerbyHandle (Variables|Transient|GetBeginAirToGroundFrameDelayTimerHandle))
  (bind _asynctask_continuoustargetingairtoground (Variables|Transient|Async|GetAsyncTask_ContinuousTargetingAirToGround)
    (:then
      (GameplayTasks|EndTask _asynctask_continuoustargetingairtoground))
    (:else)))

(event Custom|ServerNotifyDashMMEComplete
  ;[0.1115 0.05291 0.15 0.5] ServerNotifyDashMMEComplete: Dash MME has been completed on the client. Catch the server up to the events.
  (Variables|Transient|SetbClientGroundDashComplete true)
  (if (Variables|Transient|GetbLaunchPlayerInvoked)
    (Helper|CatchServerUponEventsinOrder (Utilities|Array|MakeArray "(TagName=\"MotoWrestle.Slide.AirToGround.MME.Start\")" "(TagName=\"MotoWrestle.Slide.AirToGround.MME.End\")" "(TagName=\"MotoWrestle.Slide.Ground.MME.Start\")" "(TagName=\"MotoWrestle.Slide.Ground.MME.End\")"))))

(event Custom|BeginGroundDash
  (params (FromGround :type Boolean))
  ;[0.02102 0.15 0.1313 0.5] BeginGroundDash: Dash ability with the player sliding along the ground.
  (Variables|PowerAnimationCast|SetCastMontageStartSection (select FromGround "None" "SlideDashSliding"))
  (Fortnite|FortShippingLog self "BeginGroundDash: Invoked.")
  (GameplayTasks|EndTask (Variables|Transient|Async|GetAsyncTask_PlayMontageSlideDashAir))
  ;; This should play the ground sliding montage.
  (CallFunction|PlayCastAnimation)
  (Utilities|FlowControl|Sequence
    (:then_0
      (Utilities|IsServer
        (:true
          (if (Pawn|IsBotControlled (CallFunction|GetAvatarasFortPlayerPawn))
            _))
        (:false
          _))
      (CallFunction|FindDashTarget) ; Performs target to determine what type of dash to perform
    )
    (:then_1
      ;;[0.02504 0.15 0.1486 0.5] Jump is allowed when ground sliding, bind to know when the pawn jumps.
      (bind _outfortplayerpawn (CallFunction|GetAvatarasValidFortPlayerPawn)
        (:ValidFortPlayerPawn
          (Pawn|BindEventtoOnPawnJumped _outfortplayerpawn (EventDispatchers|CreateEvent :!event HandlePawnJumpedDuringGroundDash)))))
    (:then_2
      ;;[6.6e-05 0.1323 0.15 0.5] Add controller rumble while sliding
      (bind _outfortplayerpawn_2 (CallFunction|GetAvatarasValidFortPlayerPawn)
        (:ValidFortPlayerPawn
          (Utilities|IsValid (Fortnite|GetFortPlayerControllerfromActor _outfortplayerpawn_2)
            (:Is%20Valid
              (Game|Feedback|LocalPlayForceFeedback (Fortnite|GetFortPlayerControllerfromActor _outfortplayerpawn_2) "/MotoWrestleGameplay/Gameplay/Abilities/Slide/ForceFeedback/FF_MotoWrestle_Slide_Sliding.FF_MotoWrestle_Slide_Sliding" false "None"))))))))

(event EventPerformDashToTarget
  (params (DashToTarget :type Actor%20Object%20Reference))
  ;[0.02102 0.15 0.1313 0.5] BeginGroundDash: Dash ability with the player sliding along the ground.
  ;;[0.09187 0.01364 0.09896] Activate dash to the desired target actor.
  (bind _returnvalue (Game|ApplyGuidedMotionMME (CallFunction|GetAvatarasFortPlayerPawn) "/MotoWrestleGameplay/Gameplay/Abilities/Slide/MME/FMM_GuidedMotion_MotoWrestle_DashToActor.FMM_GuidedMotion_MotoWrestle_DashToActor_C" DashToTarget))
  (Utilities|IsValid _returnvalue
    (:Is%20Valid
      (Variables|Transient|SetMovementModeRuntimeData _returnvalue)
      (Variables|Transient|SetbUsedMoveToActorMME true)
      (Fortnite|FortShippingLog self "BeginGroundDash: MoveToActor MME called."))
    (:Is%20Not%20Valid
      (Ability|EndAbility)
      ;;;[1 0 0] Warning: MME failed to execute.
      (Fortnite|FortShippingLog self "LaunchPlayer: Move To Target Actor MME failed to execute." true))))

(event EventPerformStraightDash
  (bind _returnvalue (CallFunction|GetAvatarasFortPlayerPawn))
  ;[0.02102 0.15 0.1313 0.5] BeginGroundDash: Dash ability with the player sliding along the ground.
  ;;[0.0219 0.1198 0.1018] If no target found, execute a constant force root motion MME for the dash.
  (bind _returnvalue_2 (Game|ApplyGuidedMotionMME (CallFunction|GetAvatarasFortPlayerPawn) "/MotoWrestleGameplay/Gameplay/Abilities/Slide/MME/FMM_GuidedMotion_MotoWrestle_Dash.FMM_GuidedMotion_MotoWrestle_Dash_C"))
  (Utilities|IsValid _returnvalue_2
    (:Is%20Valid
      (Variables|Transient|SetMovementModeRuntimeData _returnvalue_2)
      (Fortnite|FortShippingLog self "BeginGroundDash: Constant Force MME called.")
      ;;;[0.1443 0.01364 0.15 0.5] Sometimes the MME can fail to start.
      ;;; Invoke the gameplay events to continue ability progression.
      (Ability|Tasks|WaitDelay (+ (Math|Float|SafeDivide (ScalableFloat|GetValueasDouble (Variables|Hotfixes|GetHotfix_TargetingLength)) (ScalableFloat|GetValueatLevel (Variables|Hotfixes|GetHotfix_DashSpeed))) 0.1)
        (:OnFinish
          (Ability|SendGameplayEventtoActor _returnvalue (Variables|Settings|GetMME%20Start%20Tag))
          (Ability|SendGameplayEventtoActor _returnvalue (Variables|Settings|GetMME%20End%20Tag)))))
    (:Is%20Not%20Valid
      (Ability|EndAbility)
      ;;;[1 0 0] Warning: MME failed to execute.
      (Fortnite|FortShippingLog self "BeginGroundDash: Constant Force MME failed to execute." true))))

(event Custom|BeginGroundDashTargeting
  ;[0.1062 0.01917 0.1094] ForwardTargeting: Look for potential targets in front of the player.
  (bind (_asynctaskproxy _targetdata) (Ability|Tasks|DoTargetSelectionContinuous "GroundContinuousTargeting" "(TagName=\"MotoWrestle.Slide.EffectContainer.Application.Ground\")" false)
    (:then
      (bind _output_get (Variables|Transient|Async|SetAsyncTask_ContinuousTargeting_Ground _asynctaskproxy)))
    (:Targeted
      (if (Utilities|Array|IsNotEmpty (Ability|TargetData|GetAllActorsfromTargetData _targetdata))
        (bind (_asynctaskproxy_2 _targetdata_2) (Ability|Tasks|PerformFilteringAbilityTask "SlopeFilter" (Variables|Targeting|GetSlopeTraversalFilterTargetingPreset) false false false (Ability|TargetData|GetAllActorsfromTargetData _targetdata))
          (:then
            (Variables|Transient|Async|SetAsyncTask_SlopeFilter _asynctaskproxy_2)
            (GameplayTasks|EndTask _output_get))
          (:Targeted
            (Variables|Transient|SetContinuousTargetingHitResult (Ability|TargetData|GetHitResultfromTargetData _targetdata_2))
            (Fortnite|FortShippingLog self (Utilities|String|Append "GroundContinuousTargeting: Successfully hit actor: " (Utilities|GetDisplayName (.HitActor (Collision|BreakHitResult (Ability|TargetData|GetHitResultfromTargetData _targetdata_2))))))
            (bind _asfort_movement_comp_character (Utilities|Casting|CastToFortMovementComp_Character (Variables|/Script/Engine.Character|GetCharacterMovement (Variables|Default|GetPlayerPawn))))
            (Class|FortMovementCompCharacter|CancelActiveExtensionwithTag _asfort_movement_comp_character (GameplayTags|MakeGameplayTagContainerfromTag "(TagName=\"MovementMode.Custom.RootMotion.MotoWrestle.Slide.Ground\")"))))))))

(event Custom|EndGroundDashTargeting
  ;[0.1062 0.01917 0.1094] ForwardTargeting: Look for potential targets in front of the player.
  (GameplayTasks|EndTask (Variables|Transient|Async|GetAsyncTask_ContinuousTargeting_Ground))
  (Utilities|Time|ClearandInvalidateTimerbyHandle (Variables|Transient|GetBeginTargetingFrameDelayTimerHandle)))

(event EventCastMontageNotifyStart
  (params (NotifyTag :type Gameplay%20Tag%20Structure))
  ;[0.01664 0.1007 0.1094] NotifyStart: When the impact notification is received, perform the damage execution.
  (|CallParentFunction NotifyTag)
  (if (== NotifyTag "(TagName=\"MotoWrestle.AnimNotify.Slide.Impact\")")
    ;; Performs the target selection that eventually applies the damage of the attack.
    (bind (_asynctaskproxy _targetdata _applicationtag) (Ability|Tasks|DoTargetSelection "DamageTargetSelection" "(TagName=\"MotoWrestle.AnimNotify.Slide.Impact\")" true)
      (:then
        (Variables|Transient|Async|SetAsyncTask_DamageTargetSelection _asynctaskproxy)
        (if (not (Variables|Transient|GetbUsedMoveToActorMME))
          (bind _outfortplayerpawn (CallFunction|GetAvatarasValidFortPlayerPawn)
            (:ValidFortPlayerPawn
              (bind _asfort_movement_comp_character (Utilities|Casting|CastToFortMovementComp_Character (Variables|/Script/Engine.Character|GetCharacterMovement _outfortplayerpawn)))
              (Class|FortMovementCompCharacter|CancelActiveExtensionwithTag _asfort_movement_comp_character (GameplayTags|MakeGameplayTagContainerfromTag "(TagName=\"MovementMode.Custom.RootMotion.MotoWrestle.Slide.Ground\")"))))))
      (:Targeted
        (if (not (Variables|Transient|GetbDamageDone))
          (Variables|Transient|SetbDamageDone true)
          (Variables|Transient|SetCachedDamageApplicationTag _applicationtag)
          (bind _filteredtargetdata (Ability|FilterTargetData _targetdata))
          (Variables|Transient|SetCachedDamageTargetDataHandle _filteredtargetdata)
          (Utilities|FlowControl|DelayUntilNextTick) ; Perf
          (Ability|ApplyGameplayEffectContainer (Variables|Transient|GetCachedDamageTargetDataHandle) (Variables|Transient|GetCachedDamageApplicationTag)))))
    (elif (== NotifyTag "(TagName=\"MotoWrestle.AnimNotify.Slide.AllowMelee.Server\")")
      ;;[0.15 0.02207 0.1374 0.5] At a certain point in the dash attack, the player is allowed to execute their mantis attacks again.
      (Ability|HasAuthority
        (:True
          _))
      (elif (== NotifyTag "(TagName=\"MotoWrestle.AnimNotify.Slide.AllowMelee.Client\")")
        (Ability|HasAuthority
          (:False
            _)))))
  (Ability|RemoveGameplayEffectFromOwnerWithHandle (Variables|Transient|GetBlockMeleeGameplayEffectyHandle)))

(event EventEndDash
  ;[0.15 0.03982 0.1114 0.5] EndDash: Ground dash has ended. Jump to the appropriate montage section. Remove binds and camera modifiers.
  (if (not (Variables|Transient|GetbDashEnded))
    (|CallParentFunction)
    (Fortnite|FortShippingLog self "DashEnded: Invoked: Jumping Montage Section.")
    (bind _outfortplayerpawn (CallFunction|GetAvatarasValidFortPlayerPawn)
      (:ValidFortPlayerPawn
        (Utilities|IsValid (Fortnite|GetFortPlayerControllerfromActor _outfortplayerpawn)
          (:Is%20Valid
            ;; Camera shake at the end of the slide
            (Game|Feedback|ClientStartCameraShake :self (Fortnite|GetFortPlayerControllerfromActor _outfortplayerpawn) :Shake "/MotoWrestleGameplay/Gameplay/Abilities/Slide/Camera/CameraShake_MotoWrestle_SlideEnd.CameraShake_MotoWrestle_SlideEnd_C" :Scale 1.0 :PlaySpace "CameraLocal"))
          (:Is%20Not%20Valid
            (if (CallFunction|IsAI)
              _)))
        ;; Cleanup to end sliding GCNL
        (GameplayCue|RemoveGameplayCue _outfortplayerpawn (Variables|GameplayCue|GetGC_SlideAudio))
        ;; Break spikes
        (bind _icicles_bga (Variables|Transient|GetIcicles%20BGA)
          (:then
            (Gameplay|SilentDie _icicles_bga false :!class /Script/FortniteGame.BuildingActor))
          (:else
            _))
        (bind _bclosetoground (Helper|TraceClosetoGround false (ScalableFloat|GetValueatLevel (Variables|Hotfixes|GetHotfix_GroundDistanceCheck))))
        (Ability|Animation|MontageJumptoSection (select _bclosetoground "SlideDashGroundExit" "SlideDashAirExit"))
        (Class|FortCameraModifierTargetRotationFrustum|RemoveSelffromCameraManager (Variables|Transient|GetGroundDashCameraModifierHandle)) ; Give player control back before they land
        ;; End controller rumble if slide ended early
        (bind _outfortplayerpawn_2 (CallFunction|GetAvatarasValidFortPlayerPawn)
          (:ValidFortPlayerPawn
            (Utilities|IsValid (Fortnite|GetFortPlayerControllerfromActor _outfortplayerpawn_2)
              (:Is%20Valid
                (Game|Feedback|LocalStopForceFeedback (Fortnite|GetFortPlayerControllerfromActor _outfortplayerpawn_2) "/MotoWrestleGameplay/Gameplay/Abilities/Slide/ForceFeedback/FF_MotoWrestle_Slide_Sliding.FF_MotoWrestle_Slide_Sliding" "None"))
              (:Is%20Not%20Valid
                (if (CallFunction|IsAI)
                  _)))
            ;; Adding minor low friction after the dash is complete to prevent coming to an instant standstill when on the ground..
            (Ability|ApplyGameplayEffectToOwner "/MotoWrestleGameplay/Gameplay/Abilities/Slide/GameplayEffects/GE_MotoWrestle_Slide_LowFrictionFinish.GE_MotoWrestle_Slide_LowFrictionFinish_C")
            (bind _playerpawn (Variables|Default|GetPlayerPawn)
              (:then
                (Pawn|UnbindEventfromOnPawnJumped _playerpawn (EventDispatchers|CreateEvent :!event HandlePawnJumpedDuringGroundDash)))
              (:else
                _))
            (Ability|HasAuthority
              (:True
                (Ability|CommitAbility)))))))))

(event Ability|EventOnEndAbility
  (params (bWasCancelled :type Boolean))
  ;[0.15 0.08202 0.02261 0.5] End Ability: Clean up.
  (Utilities|FlowControl|Sequence
    (:then_0
      (Utilities|Time|ClearandInvalidateTimerbyHandle (Variables|Transient|GetBeginAirToGroundFrameDelayTimerHandle))
      (Utilities|Time|ClearandInvalidateTimerbyHandle (Variables|Transient|GetBeginTargetingFrameDelayTimerHandle))
      (Utilities|Time|ClearandInvalidateTimerbyHandle (Variables|Transient|GetInputPressedExecuteGroundDashAttackMinDelayTimerHandle))
      (GameplayTasks|EndTask (Variables|Transient|Async|GetAsyncTask_ContinuousTargeting_Ground))
      (CallFunction|EndAirtoGroundTargeting)
      (Ability|EndAirtoGroundMME)
      (Fortnite|FortShippingLog self "EndAbility: Invoked")
      (bind _outfortplayerpawn (CallFunction|GetAvatarasValidFortPlayerPawn)
        (:ValidFortPlayerPawn
          (Utilities|IsValid (Fortnite|GetFortPlayerControllerfromActor _outfortplayerpawn)
            (:Is%20Valid
              ;; Give player control back before they land
              (bind _playercameramanager (Variables|/Script/Engine.PlayerController|GetPlayerCameraManager (Fortnite|GetFortPlayerControllerfromActor _outfortplayerpawn))
                (:then
                  (Game|Player|RemoveCameraModifier _playercameramanager (Variables|Transient|GetGroundDashCameraModifierHandle))
                  (Game|Player|RemoveCameraModifier _playercameramanager (Variables|Transient|GetAirToGroundCameraModifierHandle))
                  (Ability|RemoveGameplayEffectFromOwnerWithHandle (Variables|Transient|GetBlockMeleeGameplayEffectyHandle))
                  (bind _playerpawn (Variables|Default|GetPlayerPawn)
                    (:then
                      (Pawn|UnbindEventfromOnPawnJumped _playerpawn (EventDispatchers|CreateEvent :!event HandlePawnJumpedDuringGroundDash)))
                    (:else)))
                (:else)))))))
    (:then_1
      (|CallParentFunction bWasCancelled))))

(fn AirToGroundCameraModHeightCheck
  (bind _hotfix_minheightforairtogroundcameramod (Variables|Hotfixes|GetHotfix_MinHeightForAirToGroundCameraMod))
  (bind _returnvalue (ScalableFloat|GetValueatLevel _hotfix_minheightforairtogroundcameramod))
  ; Returns the player is far enough from the ground for the air to ground camera mod to play.
  (bind _outfortplayerpawn (CallFunction|GetAvatarasValidFortPlayerPawn)
    (:ValidFortPlayerPawn
      (if (< _returnvalue 0)
        (return false)
        (else
          (bind (_outhit _returnvalue_2) (Collision|LineTraceForObjects (Transformation|GetActorLocation _outfortplayerpawn) (Utilities|Operators|Subtract (Transformation|GetActorLocation _outfortplayerpawn) (Math|Vector|MakeVector 0.0 0.0 _returnvalue)) (Utilities|Array|MakeArray "ObjectTypeQuery1" "ObjectTypeQuery2" "ObjectTypeQuery5" "ObjectTypeQuery13" "ObjectTypeQuery4" "ObjectTypeQuery6")))
          (return (not _returnvalue_2)))))))

(fn ApplyAirToGroundDiveCameraModifier
  ; ApplyAirToGroundDiveCameraModifier: Adding a camera modifer that has the camera look down in the direction of the dive.
  (Ability|HasAuthority
    (:False
      (bind _returnvalue (Camera|AirtoGroundCameraModHeightCheck))
      (if _returnvalue
        (bind _returnvalue_2 (Fortnite|Camera|AddCameraModifierforOwningPlayerController (Variables|Camera|GetCameraModifier_AirToGroundDive) (Ability|GetAvatarActorfromActorInfo)))
        (bind _output_get (Variables|Transient|SetAirToGroundCameraModifierHandle (Utilities|Casting|CastToFortCameraModifier_TargetRotationFrustum _returnvalue_2)))
        (bind _ascamera_modifier_frustrum_moto_wrestle_air_to_ground_dash (Utilities|Casting|CastToCameraModifierFrustrum_MotoWrestle_AirToGroundDash _output_get))
        (Class|CameraModifierFrustrumMotoWrestleAirtoGroundDash|RemoveAfterDelay _ascamera_modifier_frustrum_moto_wrestle_air_to_ground_dash (ScalableFloat|GetValueatLevel (Variables|Hotfixes|GetHotfix_AirToGroundCameraModifierDuration)))))))

(fn ApplyGroundDashCameraModifier
  ; ApplyAirToGroundDiveCameraModifier: Adding a camera modifer that has the camera look down in the direction of the dive.
  (Ability|HasAuthority
    (:False
      (bind _returnvalue (Fortnite|Camera|AddCameraModifierforOwningPlayerController (Variables|Camera|GetCameraModifierGroundDash) (Ability|GetAvatarActorfromActorInfo)))
      (bind _output_get (Variables|Transient|SetGroundDashCameraModifierHandle (Utilities|Casting|CastToFortCameraModifier_TargetRotationFrustum _returnvalue)))
      (bind _ascamera_modifier_frustrum_moto_wrestle_ground_dash (Utilities|Casting|CastToCameraModifierFrustrum_MotoWrestle_GroundDash _output_get))
      (Class|CameraModifierFrustrumMotoWrestleGroundDash|RemoveAfterDelay _ascamera_modifier_frustrum_moto_wrestle_ground_dash (ScalableFloat|GetValueasDouble (Variables|Hotfixes|GetHotfix_GroundDashCameraModifierDuration))))))

(fn CatchServerUpOnEventsInOrder
  (params (EventGameplayTags :type Array%20of%20Gameplay%20Tag%20Structures))
  ;[0.03741 0.15 0.1457 0.5] CatchServerUpOnEventsInOrder: Send all the gameplay events to the player in order.
  (Ability|HasAuthority
    (:True
      (for _array_element EventGameplayTags
        (Ability|SendGameplayEventtoActor (CallFunction|GetAvatarasFortPlayerPawn) _array_element)))))

(fn EndAirToGroundMME
  ;[0.05533 0.1481 0.15 0.5] Ends the air to ground MME.
  ; Runs successfully on both server and client.
  (bind _playerpawn (Variables|Default|GetPlayerPawn)
    (:then
      (Ability|HasAuthority
        (:True
          (if (Pawn|IsBotControlled _playerpawn)
            _))
        (:False
          _))
      (bind _asfort_movement_comp_character (Utilities|Casting|CastToFortMovementComp_Character (Variables|/Script/Engine.Character|GetCharacterMovement _playerpawn)))
      (Class|FortMovementCompCharacter|CancelActiveExtensionwithTag _asfort_movement_comp_character (GameplayTags|MakeGameplayTagContainerfromTag "(TagName=\"MovementMode.Custom.RootMotion.MotoWrestle.Slide.AirToGround\")")))
    (:else)))

(fn ExecuteFailGameplayCue
  (params (Target :type Actor%20Object%20Reference))
  (GameplayCue|ExecuteGameplayCue%28Burst%29Local Target (Variables|GameplayCue|GetGameplayCueTagActivateFailed) (Ability|GameplayCue|MakeGameplayCueParameters 0.0)))

(fn FilterTargetData
  (params (TargetData :type Gameplay%20Ability%20Target%20Data%20Handle%20Structure))
  ; FilterTargetData: Removing duplicate actor entries.
  (for _index (range (Ability|TargetData|GetDataCountfromTargetData TargetData))
    (Utilities|IsValid (.HitActor (Collision|BreakHitResult (Ability|TargetData|GetHitResultfromTargetData TargetData _index)))
      (:Is%20Valid
        (bind _returnvalue (Utilities|Array|AddUnique (LocalVariables|GetHitActors) (.HitActor (Collision|BreakHitResult (Ability|TargetData|GetHitResultfromTargetData TargetData _index)))))
        (if (>= _returnvalue 0)
          (Fortnite|FortShippingLog self (Utilities|String|Append " FilterTargetData:  " (Utilities|GetDisplayName (.HitActor (Collision|BreakHitResult (Ability|TargetData|GetHitResultfromTargetData TargetData _index))))))
          (bind _returnvalue_2 (Ability|TargetData|AppendTargetDataHandle (LocalVariables|GetReturnHandle) (Ability|TargetData|AbilityTargetDatafromHitResult (Ability|TargetData|GetHitResultfromTargetData TargetData _index))))
          (LocalVariables|SetReturnHandle _returnvalue_2)
          (for _array_element (LocalVariables|GetHitActors)
            (GameplayCue|ExecuteGameplayCue _array_element (Variables|GameplayCue|GetGC%20Targets%20Hit)))))))
  (return (LocalVariables|GetReturnHandle)))

(fn GetCustomAbilitySourceTransform
  (bind _returnvalue (CallFunction|GetAvatarasFortPlayerPawn))
  (bind (_returnvalue_location _returnvalue_rotation _returnvalue_scale) (Math|Transform|BreakTransform (Transformation|GetActorTransform _returnvalue)))
  (bind _returnvalue_2 (CallFunction|GetAvatarasFortPlayerPawn))
  (bind (_returnvalue_location_2 _returnvalue_rotation_2 _returnvalue_scale_2) (Math|Transform|BreakTransform (Transformation|GetActorTransform _returnvalue_2)))
  ;[0.1044 0.01678 0.1146] GetCustomAbilitySourceTransform: This function is used to return the source transform for the gameplay effect containerof 'Gameplay.Application.hit', which is used during continuous targeting while the dash takes place.
  ; IF the player is moving, this returns the velocity direction as the rotation.
  (if (Variables|Transient|GetbAirToGroundMMERunning)
    (Utilities|IsValid _returnvalue
      (:Is%20Valid
        (return _returnvalue_location (Math|Conversions|RotationFromXVector (Math|Vector|VectorDown)) _returnvalue_scale)))
    (else
      (Utilities|IsValid _returnvalue_2
        (:Is%20Valid
          (return _returnvalue_location_2 (Math|Rotator|SelectRotator (Math|Rotator|MakeRotfromX (Transformation|GetVelocity _returnvalue_2)) (Transformation|GetActorRotation _returnvalue_2) (> (Math|Vector|VectorLength (Transformation|GetVelocity _returnvalue_2)) 0)) _returnvalue_scale_2))
        (:Is%20Not%20Valid
          ;;[1 0 0] Warning: Avatar player pawn is NULL.
          (Fortnite|FortShippingLog self "GetCustomAbilitySourceTransform: Avatar player pawn is NULL." true)
          (return "0, 0, 0" "0, 0, 0" "1.000000,1.000000,1.000000"))))))

(fn HandlePawnJumpedDuringGroundDash
  ;[0.0476 0.09568 0.15 0.5] HandlePawnJumpedDuringGroundDash: If the pawn jumps during the ground dash, the ability should end and launch the player forward.
  (Fortnite|FortShippingLog self "HandlePawnJumpedDuringGroundDash: Invoked")
  (Ability|EndAbility)
  (bind _outfortplayerpawn (CallFunction|GetAvatarasValidFortPlayerPawn)
    (:ValidFortPlayerPawn
      (bind (_returnvalue_x _returnvalue_y _returnvalue_z) (Math|Vector|BreakVector (Math|Vector|VectorClampSize2D (Transformation|GetVelocity _outfortplayerpawn) 0.0 1e+03)))
      (Pawn|Character|LaunchCharacterJump _outfortplayerpawn (Math|Vector|MakeVector _returnvalue_x _returnvalue_y (Math|Float|Max(Float) _returnvalue_z 5e+02)) true true true true))))

(fn CanActivateAbility
  (params (ActorInfo :type Gameplay%20Ability%20Actor%20Info%20Structure) (Handle :type Gameplay%20Ability%20Spec%20Handle%20Structure))
  (bind (_relevanttags _returnvalue) (|CallParentFunction ActorInfo Handle))
  (return _relevanttags _returnvalue))

(fn OverrideFailedReason
  (params (FailedReason :type Gameplay%20Tag%20Container%20Structure%20%28by%20ref%29))
  (Utilities|FlowControl|Sequence
    (:then_0
      (bind _returnvalue (Ability|CheckAbilityCost))
      (if _returnvalue
        _
        (elif (<= (FortWeapon|GetMagazineAmmoCount (Ability|GetCurrentSourceWeapon)) 0)
          (Default|ExecuteFailGameplayCue (CallFunction|GetAvatarasFortPlayerPawn))
          (GameplayTags|AddGameplayTag (LocalVariables|GetReturnGameplayTagContainer) "(TagName=\"Activate.Failed.Cooldown\")" :!class /Script/GameplayTags.BlueprintGameplayTagLibrary))))
    (:then_1
      (return (LocalVariables|GetReturnGameplayTagContainer)))))

(fn PreparePlayerForLaunch
  (|CallParentFunction)
  ; Block melee.
  (bind _returnvalue (Ability|ApplyGameplayEffectToOwner "/MotoWrestleGameplay/Gameplay/Abilities/Slide/GameplayEffects/GE_MotoWrestle_Slide_BlockMelee.GE_MotoWrestle_Slide_BlockMelee_C"))
  (Variables|Transient|SetBlockMeleeGameplayEffectyHandle _returnvalue)
  (Camera|ApplyAirtoGroundDiveCameraModifier))

(fn ResetTransientState
  ; Reset transient values that should not hold cached values between air dash activations.
  (|CallParentFunction)
  (Variables|Transient|SetAirToGroundMMEStartTriggered false)
  (Variables|Transient|SetAirToGroundMMEEndTriggered false)
  (Variables|Transient|SetContinuousTargetingHitResult)
  (Variables|Transient|SetbAirToGroundMMERunning false)
  (Variables|Transient|SetCachedDamageApplicationTag)
  (Variables|Transient|SetCachedDamageTargetDataHandle)
  (Variables|Transient|SetbInputPressedToExecuteGroundDashAttack false)
  (Variables|Transient|SetAirToGroundDurationFrameCount 0)
  (Variables|Transient|SetbClientGroundDashComplete false)
  (Variables|Transient|SetbDamageDone false)
  (Variables|Transient|SetbLaunchPlayerInvoked false))

(fn ShouldPlayFailedMontage
  (params (FailedReason :type Gameplay%20Tag%20Container%20Structure%20%28by%20ref%29))
  (return (GameplayTags|HasTag FailedReason "(TagName=\"Activate.Failed.Cooldown\")")))

(fn SpawnIciclesBGA
  (bind (_returnvalue_location _returnvalue_rotation) (Math|Transform|BreakTransform (Transformation|GetActorTransform (CallFunction|GetAvatarasFortPlayerPawn))))
  (Utilities|IsServer
    (:true
      (bind _returnvalue (Building|SpawnBuildingGameplayActor "/MotoWrestleGameplay/Gameplay/Abilities/Slide/GameplayAbilities/BGA_MotoWrestle_Icicles.BGA_MotoWrestle_Icicles_C" (Math|Transform|MakeTransform _returnvalue_location _returnvalue_rotation) (CallFunction|GetAvatarasFortPlayerPawn) true true))
      (Variables|Transient|SetIcicles%20BGA _returnvalue))))

(fn TraceCloseToGround
  (params (bSkipUsingTargetingHitResult :type Boolean) (GroundCheckDistance :type Float%20%28double-precision%29))
  (bind _returnvalue (CallFunction|GetAvatarasFortPlayerPawn))
  (bind _returnvalue_2 (Transformation|GetActorLocation _returnvalue))
  (bind _returnvalue_3 (Math|Vector|GetUnitDirection(Vector) _returnvalue_2 (.ImpactPoint (Collision|BreakHitResult (Variables|Transient|GetContinuousTargetingHitResult)))))
  (bind _continuoustargetinghitresult (Variables|Transient|GetContinuousTargetingHitResult))
  (bind (_bblockinghit _binitialoverlap _time _distance _location _impactpoint _normal _impactnormal _physmat _hitactor _hitcomponent) (Collision|BreakHitResult _continuoustargetinghitresult))
  (bind _returnvalue_4 (CallFunction|GetAvatarasFortPlayerPawn))
  ;[0.03714 0.1414 0.15 0.5] TraceCloseToGround: Return if the player is close to the ground.
  (bind _outfortplayerpawn (CallFunction|GetAvatarasValidFortPlayerPawn)
    (:ValidFortPlayerPawn
      (Variables|/Script/Engine.Character|GetCapsuleComponent _outfortplayerpawn
        (:then
          (Utilities|FlowControl|Sequence
            (:then_0
              (if (not bSkipUsingTargetingHitResult)
                (Utilities|IsValid _hitcomponent
                  (:Is%20Valid
                    ;;[0.15 0.03196 0.1314 0.5] Runs a line component trace (very cheap trace) to predict where the dash is going to end up taking the player so that the ground trace can be run from that found location.
                    (bind (_hitlocation _hitnormal _bonename _outhit _returnvalue_5) (Collision|LineTraceComponent _hitcomponent _returnvalue_2 (+ _returnvalue_2 (* _returnvalue_3 (LocalVariables|GetLineTraceToHitCompDistance))) true false false))
                    (if _returnvalue_5
                      (bind _output_get (LocalVariables|SetTraceStartLocation (+ _hitlocation (* (* _returnvalue_3 -1.0) (+ (Components|Capsule|GetScaledCapsuleRadius (Variables|/Script/Engine.Character|GetCapsuleComponent _returnvalue)) 1e+01)))))
                      (LocalVariables|SetTraceEndLocation (Utilities|Operators|Subtract _output_get (Math|Vector|MakeVector 0.0 0.0 (Variables|GetGroundCheckDistance))))
                      (LocalVariables|SetbPerformTrace true)))
                  (:Is%20Not%20Valid
                    _)))
              (bind _output_get_2 (LocalVariables|SetTraceStartLocation (Transformation|GetActorLocation _returnvalue_4)))
              (LocalVariables|SetTraceEndLocation (Utilities|Operators|Subtract _output_get_2 (Math|Vector|MakeVector 0.0 0.0 (+ (Components|Capsule|GetScaledCapsuleHalfHeight (Variables|/Script/Engine.Character|GetCapsuleComponent _returnvalue_4)) (Variables|GetGroundCheckDistance)))))
              (LocalVariables|SetbPerformTrace true))
            (:then_1
              (if (LocalVariables|GetbPerformTrace)
                (bind (_outhit_2 _returnvalue_6) (Collision|LineTraceByChannel (LocalVariables|GetTraceStartLocation) (LocalVariables|GetTraceEndLocation) "TraceTypeQuery10" true (Utilities|Array|MakeArray (CallFunction|GetAvatarasFortPlayerPawn)) "None" true "(R=1.000000,G=0.000000,B=0.000000,A=1.000000)" "(R=0.000000,G=1.000000,B=0.000000,A=1.000000)" 1e+01))
                (return _returnvalue_6)))))
        (:else)))
    (:AvatarCastFailed
      (return false))))
